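# Niagara effect for the GoneFishing guitar cosmetic, producing smoke and grass clippings at an exhaust and swirling grass and wind at its blades.
# GPU sprite and CPU mesh grass spray along the exhaust direction, spawning faster with player speed.
# Flipbook smoke uses two spawn groups for separate plumes, with alpha tied to speed.
# At the blades, sine/cosine-driven GPU grass orbits alongside rotating whirl meshes tinted by time of day, with a front-end preview override.
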NiagaraSystem /SparksCosmetics/Instrument/Guitar/GoneFishing/Effects/NS_Guitar_GoneFishing_GrassExhaust
dependencies {
 usedDataInterfaces [&NiagaraDataInterfaceCurve, &NiagaraDataInterfaceColorCurve]
 usedDynamicInputs [
  &RandomRangeLinearColor
  &NormalizeVector
  &RemapRange
 ]
 usedModules [
  &self:SystemSpawnScript.SystemScriptSource.SystemScriptGraph.NiagaraNodeAssignment_0.SetVariables_#1
  &self:SystemSpawnScript.SystemScriptSource.SystemScriptGraph.NiagaraNodeAssignment_2.SetVariables_#2
  &SystemState
  &self:SystemSpawnScript.SystemScriptSource.SystemScriptGraph.NiagaraNodeAssignment_1.SetVariables_#3
 ]
}
emitters {
 ExhaustSmoke {
  emitterUpdateScript [
   &EmitterState()
   &SpawnRate(SpawnRate=15, Use Spawn Probability=true)
   &SpawnRate(SpawnGroup=1, SpawnRate=10, Use Spawn Probability=true) {module=SpawnRate002}
   &SpawnRate(
    Spawn Probability=&FloatFromCurve(FloatCurve=DataInterface() {propertyValues {Curve {keys [<at 0.4 1>,<at 0.6 0>]}}}, CurveIndex=$Emitter.NormalizedLoopAge)
    SpawnRate=25
    Use Spawn Probability=true
   ) {enabled=false, module=SpawnRate001}
  ]
  particleSpawnScript [
   &InitializeParticle(
    Color=&LerpLinearColors(StartColor=<rgba 0.75 0.75 1 0.18>, EndColor=<rgba 0.75 0.75 1 0.25>, LerpFactor=$System.NormalizedPlayerSpeed)
    Lifetime=&RandomRangeFloat(Minimum=0.5, Maximum=0.75)
    Position Offset=&MaskVectorBySpawnGroup(Masked Vector=$System.ExhaustOffset, Passthrough Vector=<0 1 -5>)
    Position Offset Coordinate Space=::Local
    Sprite Rotation Mode=::NewEnumerator2
    Sprite Size Mode=::NewEnumerator3
    Sprite UV Mode=::NewEnumerator4
    Sprite UV Scale:Vector2f=&MaskVector2dBySpawnGroup(Masked Vector2D=<1 1>, Passthrough Vector2D=<-1 1>, Spawn Group Mask=0)
    Uniform Sprite Size=&RandomRangeFloat(Minimum=20, Maximum=30)
    UsePositionOffset=true
    Write Material Random=true
   )
   &AddVelocity(Velocity=&MaskVectorBySpawnGroup(Masked Vector=&Multiply_VectorByFloat(Vector=$System.ExhaustFacing, Float=60), Passthrough Vector=<0 35 0>))
   /SetVariables(Particles.SubUVRandomStart:NiagaraInt32=&RandomRangeInt(Minimum=0, Maximum=44)) {isSetParameters=true}
  ]
  particleUpdateScript [
   &ParticleState()
   &ScaleColor(
    Scale Alpha=&FloatFromCurve(
     FloatCurve=DataInterface() {propertyValues {Curve {keys [
         {arriveTangent=8.9237, interpMode=RCIM_Cubic, leaveTangent=8.9237, tangentMode=RCTM_User, time=0, value=0}
         {arriveTangent=-3.3421, interpMode=RCIM_Cubic, leaveTangent=-3.3421, tangentMode=RCTM_User, time=0.2, value=1}
         <at 1 0 RCIM_Cubic>
        ]}}}
     Scale Curve=0.5
    )
    ScaleA=true
    ScaleRGB=false
   )
   &SpriteRotationRate(Rotation Rate=&FloatFromCurve(
     FloatCurve=DataInterface() {propertyValues {Curve {keys [
         {arriveTangent=-4.0107, interpMode=RCIM_Cubic, leaveTangent=-4.0107, tangentMode=RCTM_User, time=-0.0046, value=1.0113}
         {arriveTangent=-0.2496, interpMode=RCIM_Cubic, leaveTangent=-0.2496, tangentMode=RCTM_User, time=0.5, value=0.16}
        ]}}}
     Scale Curve=&Multiply_Float(
      A=&RandomRangeFloat(Minimum=-300, Maximum=300)
      B=&MaskFloatBySpawnGroup(Masked Float=0.5, Passthrough Float=1)
     )
    ))
   &ScaleSpriteSize(Scale Factor=&Vector2DFromFloat(Value=&FloatFromCurve(FloatCurve=DataInterface() {propertyValues {Curve {keys [
          {arriveTangent=1.4788, interpMode=RCIM_Cubic, leaveTangent=1.4788, tangentMode=RCTM_User, time=0, value=0.4}
          {arriveTangent=0.2754, interpMode=RCIM_Cubic, leaveTangent=0.2754, tangentMode=RCTM_User, time=1.0463, value=1}
         ]}}})))
   &Drag(Drag=3, Use Linear Drag=true, Use Rotational Drag=false)
   &GravityForce(Gravity=<0 0 -50>)
   &SolveForcesAndVelocity(Clamp Velocity=false, Limit Acceleration=false)
   &SubUVAnimation(
    End Frame:NiagaraInt32=&Add_Integer(A=$Particles.SubUVRandomStart, B=23)
    Start Frame:NiagaraInt32=$Particles.SubUVRandomStart
    SubUV Animation Mode:ENiagara_SubUVLookupMode=::NewEnumerator1
    SubUV Curve=DataInterface() {propertyValues {Curve {keys [
        {arriveTangent=1.6197, interpMode=RCIM_Cubic, leaveTangent=1.6197, tangentMode=RCTM_User, time=0, value=0}
        {arriveTangent=0.5634, interpMode=RCIM_Cubic, leaveTangent=0.5634, tangentMode=RCTM_User, time=1, value=1}
       ]}}}
   )
  ]
  rendererClasses [&NiagaraSpriteRendererProperties]
  renderers [{rendererClass=&NiagaraSpriteRendererProperties}]
 }
 FN_Debug {
  bEnabled=false
  emitterUpdateScript [
   &EmitterState()
   &SpawnRate(Use Spawn Probability=false)
   &SpawnBurst_Instantaneous(Spawn Count=1, Use Spawn Probability=false) {enabled=false}
  ]
  particleSpawnScript [
   /Niagara/Modules/Spawn/Initialization/InitializeParticle(
    Module.Write Color=true
    Module.Write Lifetime=true
    Module.Write Mass=true
    Module.Write Material Random=true
    Module.Write Position=false
    Module.Write Scale=false
    Module.Write SpriteRotation=false
    Module.Write SpriteSize=true
    Position=&SimulationPosition()
    Sprite Size=<100 100>
   )
   &DynamicMaterialParameters(
    Index 0 Param 1=$System.NormalizedPlayerSpeed
    Module.Param0WriteEnabled=true
    Module.Param1WriteEnabled=true
    Module.Param2WriteEnabled=true
    Module.Param3WriteEnabled=true
   ) {module=DynamicMaterialParameters001}
  ]
  particleUpdateScript [
   &ParticleState()
   /SetVariables(Particles.Debug:NiagaraFloat=&VectorLength(Vector=$Engine.Owner.Velocity)) {isSetParameters=true}
   &DynamicMaterialParameters(
    Index 0 Param 1=$Particles.Debug
    Module.Param0WriteEnabled=true
    Module.Param1WriteEnabled=true
    Module.Param2WriteEnabled=true
    Module.Param3WriteEnabled=true
   ) {enabled=false}
  ]
  rendererClasses [&NiagaraSpriteRendererProperties]
  renderers [{rendererClass=&NiagaraSpriteRendererProperties}]
 }
 GrassGPU {
  simTarget=GPUComputeSim
  emitterUpdateScript [
   &EmitterState()
   &SpawnRate(
    Spawn Probability=&FloatFromCurve(FloatCurve=DataInterface() {propertyValues {Curve {keys [<at 0.25 0>,<at 0.5 1>]}}}, CurveIndex=$Engine.System.Age)
    SpawnRate=&Lerp_Float(A=25, B=60, Alpha=$System.NormalizedPlayerSpeed)
    Use Spawn Probability=true
   )
   &SpawnRate(
    Spawn Probability=&FloatFromCurve(FloatCurve=DataInterface() {propertyValues {Curve {keys [<at 0.25 0>,<at 0.5 1>]}}}, CurveIndex=$Engine.System.Age)
    SpawnRate=500
    Use Spawn Probability=true
   ) {enabled=false, module=SpawnRate001}
  ]
  particleSpawnScript [
   /SetVariables(Particles.SubImageIndex=&RandomRangeFloat(Minimum=0, Maximum=6)) {isSetParameters=true, module=SetVariables_#4}
   /Niagara/Modules/Spawn/Initialization/V2/InitializeParticle(
    Color=<rgba 0.6531 0.6531 0.6531 1>
    Lifetime Max=0.2
    Lifetime Min=0.15
    Lifetime Mode=::NewEnumerator1
    Position Offset=$System.ExhaustOffset
    Position Offset Coordinate Space=::Local
    Sprite Rotation Mode=::NewEnumerator2
    Sprite Size Mode=::NewEnumerator1
    Sprite UV Mode=::NewEnumerator3
    Uniform Sprite Size Max=6
    Uniform Sprite Size Min=4
    UsePositionOffset=true
    Write Material Random=true
   )
   &AddVelocity(Velocity=&Multiply_VectorByFloat(Vector=$System.ExhaustFacing, Float=60))
   &InheritVelocity(Module.Limit Speed=true, Velocity Scale=&RandomRangeVector(Minimum=<0.5 0.5 0.5>, Maximum=<1 1 1>)) {enabled=false}
   &AddVelocityInCone(
    Cone Angle=100
    Cone Axis=$System.ExhaustFacing
    Module.Use Velocity Falloff On Cone Axis=false
    Velocity Strength=&UniformRangedFloat(Minimum=0, Maximum=75)
   )
   &ShapeLocation(MaskSpawnGroup=false, Override Seed=false, Sphere Radius=3, UseOwnerScale=true)
  ]
  particleUpdateScript [
   &ParticleState()
   &Drag(Use Linear Drag=true, Use Rotational Drag=false)
   &GravityForce(Gravity=<0 0 -300>)
   &CurlNoiseForce(Noise Frequency=100, Noise Strength=100, Pan Noise=false)
   &SolveForcesAndVelocity(Clamp Velocity=false, Limit Acceleration=false)
   &ScaleColor(
    Scale Alpha=1.2
    Scale RGB=&MakeVectorFromLinearColorRGB(Color=&ColorFromCurve(ColorCurve=DataInterface() {propertyValues {
        AlphaCurve {keys []}
        BlueCurve {keys [<at 0.0081 0.6244>,<at 0.1423 0.6305>,<at 0.9559 1>]}
        GreenCurve {keys [<at 0.0081 0.8353>,<at 0.1423 0.7629>,<at 0.9559 0.8821>]}
        RedCurve {keys [<at 0.0081 1>,<at 0.1423 1>,<at 0.9559 0.8145>]}
       }}))
    ScaleA=true
    ScaleRGB=true
   )
   &ScaleSpriteSize(Uniform Curve Sprite Scale=DataInterface() {propertyValues {Curve {keys [
        {arriveTangent=8.5765, interpMode=RCIM_Cubic, leaveTangent=8.5765, tangentMode=RCTM_User, time=0.1, value=0.5}
        {arriveTangent=-0.5096, interpMode=RCIM_Cubic, leaveTangent=-0.5096, time=0.2, value=1}
        {arriveTangent=-5.3839, interpMode=RCIM_Cubic, leaveTangent=-5.3839, tangentMode=RCTM_User, time=0.7622, value=0.1625}
        {arriveTangent=0.095, interpMode=RCIM_Cubic, leaveTangent=0.095, tangentMode=RCTM_User, time=0.8, value=-0.05}
       ]}}})
   &SpriteRotationRate(Rotation Rate=&RandomRangeFloat(Minimum=-360, Maximum=360))
  ]
  rendererClasses [&NiagaraSpriteRendererProperties]
  renderers [{rendererClass=&NiagaraSpriteRendererProperties}]
 }
 GrassMesh {
  emitterUpdateScript [
   &EmitterState()
   &SpawnRate(
    Spawn Probability=&FloatFromCurve(FloatCurve=DataInterface() {propertyValues {Curve {keys [<at 0.25 0>,<at 0.5 1>]}}}, CurveIndex=$Engine.System.Age)
    SpawnRate=&Lerp_Float(A=5, B=20, Alpha=$System.NormalizedPlayerSpeed)
    Use Spawn Probability=true
   )
  ]
  particleSpawnScript [
   /SetVariables(Particles.MeshIndex=&RandomRangeInt(Minimum=0, Maximum=1)) {isSetParameters=true}
   &InitializeParticle(
    Color=$System.GrassBaseColor
    Lifetime=0.2
    Mesh Scale Mode=::NewEnumerator1
    Mesh Uniform Scale Max=0.09
    Mesh Uniform Scale Min=0.07
    Position Offset=$System.ExhaustOffset
    Position Offset Coordinate Space=::Local
    UsePositionOffset=true
    Write Material Random=true
   )
   &AddVelocity(Velocity=&Multiply_VectorByFloat(Vector=$System.ExhaustFacing, Float=60))
   &InheritVelocity(
    Module.Limit Speed=true
    Velocity Limit=50
    Velocity Scale=&RandomRangeVector(Minimum=<0.5 0.5 0.5>, Maximum=<1 1 1>)
   ) {enabled=false}
   &AddVelocityInCone(
    Cone Angle=50
    Cone Axis=$System.ExhaustFacing
    Module.Use Velocity Falloff On Cone Axis=false
    Velocity Strength=&UniformRangedFloat(Minimum=0, Maximum=50)
   )
   &InitialMeshOrientation(Use Rotation Vector=false)
   &ShapeLocation(MaskSpawnGroup=false, Override Seed=false, Sphere Radius=3, UseOwnerScale=true)
  ]
  particleUpdateScript [
   &ParticleState()
   &Drag(Use Linear Drag=true, Use Rotational Drag=false)
   &GravityForce(Gravity=<0 0 -300>)
   &CurlNoiseForce(Noise Frequency=100, Noise Strength=50, Pan Noise=false)
   &SolveForcesAndVelocity(Clamp Velocity=false, Limit Acceleration=false)
   &ScaleColor(
    Scale Alpha=1.2
    Scale RGB=&MakeVectorFromLinearColorRGB(Color=&ColorFromCurve(ColorCurve=DataInterface() {propertyValues {
        AlphaCurve {keys []}
        BlueCurve {keys [<at 0.0081 0.6244>,<at 0.1423 0.6305>,<at 0.9559 1>]}
        GreenCurve {keys [<at 0.0081 0.8353>,<at 0.1423 0.7629>,<at 0.9559 0.8821>]}
        RedCurve {keys [<at 0.0081 1>,<at 0.1423 1>,<at 0.9559 0.8145>]}
       }}))
    ScaleA=true
    ScaleRGB=true
   )
   &UpdateMeshOrientation(Rotation Rate=&RandomRangeFloat(Minimum=-0.4, Maximum=0.4), Rotation Vector=<1 1 1>)
   &ScaleMeshSize(Scale Factor=&VectorFromFloat(Value=&FloatFromCurve(FloatCurve=DataInterface() {propertyValues {Curve {keys [
          {arriveTangent=8.5765, interpMode=RCIM_Cubic, leaveTangent=8.5765, tangentMode=RCTM_User, time=0.1, value=0.7}
          {arriveTangent=0.65, interpMode=RCIM_Cubic, leaveTangent=0.65, time=0.2, value=1}
          {arriveTangent=-1.25, interpMode=RCIM_Cubic, leaveTangent=-1.25, time=0.6, value=0.625}
          {arriveTangent=0.095, interpMode=RCIM_Cubic, leaveTangent=0.095, tangentMode=RCTM_User, time=1, value=0}
         ]}}})))
   &ScaleSpriteSizeBySpeed(
    Max Scale Factor=&Vector2DFromFloat(Value=1)
    Min Scale Factor=&Vector2DFromFloat(Value=0.25)
    Velocity Threshold=200
   )
  ]
  rendererClasses [&NiagaraMeshRendererProperties]
  renderers [{rendererClass=&NiagaraMeshRendererProperties, rendererIndex=0}]
 }
 GrassSpinning {
  simTarget=GPUComputeSim
  emitterUpdateScript [
   &EmitterState()
   &SpawnRate(SpawnRate=30, Use Spawn Probability=false)
  ]
  particleSpawnScript [
   /SetVariables(Particles.SubImageIndex=&RandomRangeFloat(Minimum=0, Maximum=6)) {isSetParameters=true, module=SetVariables_#4}
   &InitializeParticle(
    Color=<rgba 0.6531 0.6531 0.6531 1>
    Lifetime=&UniformRangedFloat(Minimum=0.4, Maximum=1)
    Position Offset=$System.BladesOffset
    Position Offset Coordinate Space=::Local
    Sprite Rotation Mode=::NewEnumerator2
    Sprite Size Mode=::NewEnumerator1
    Sprite UV Mode=::NewEnumerator3
    Uniform Sprite Size Max=6
    Uniform Sprite Size Min=4
    UsePositionOffset=true
    Write Material Random=true
   )
   /SetVariables(Particles.StartRadius:NiagaraFloat=&RandomRangeFloat(Minimum=1, Maximum=10)) {isSetParameters=true}
  ]
  particleUpdateScript [
   &ParticleState()
   /SetVariables(Particles.StartRadius:NiagaraFloat=/Niagara/DynamicInputs/Lerp/SmoothLerpOverTime_Float(
     Convergence Rate:NiagaraFloat=2
     Smooth Value:NiagaraFloat=$Particles.StartRadius
     Target Value:NiagaraFloat=&RandomRangeFloat(Minimum=7, Maximum=14)
    )) {isSetParameters=true}
   &ScaleColor(
    Scale Alpha=1.2
    Scale RGB=&MakeVectorFromLinearColorRGB(Color=&ColorFromCurve(ColorCurve=DataInterface() {propertyValues {
        AlphaCurve {keys []}
        BlueCurve {keys [<at 0.0081 0.6244>,<at 0.1423 0.6305>,<at 0.9559 1>]}
        GreenCurve {keys [<at 0.0081 0.8353>,<at 0.1423 0.7629>,<at 0.9559 0.8821>]}
        RedCurve {keys [<at 0.0081 1>,<at 0.1423 1>,<at 0.9559 0.8145>]}
       }}))
    ScaleA=true
    ScaleRGB=true
   )
   &ScaleSpriteSize(Uniform Curve Sprite Scale=DataInterface() {propertyValues {Curve {keys [
        {arriveTangent=8.5765, interpMode=RCIM_Cubic, leaveTangent=8.5765, tangentMode=RCTM_User, time=0.1, value=0.5}
        {arriveTangent=-0.5096, interpMode=RCIM_Cubic, leaveTangent=-0.5096, time=0.2, value=1}
        {arriveTangent=-5.3839, interpMode=RCIM_Cubic, leaveTangent=-5.3839, tangentMode=RCTM_User, time=0.7622, value=0.1625}
        {arriveTangent=0.095, interpMode=RCIM_Cubic, leaveTangent=0.095, tangentMode=RCTM_User, time=0.8, value=-0.05}
       ]}}})
   &SpriteRotationRate(Rotation Rate=&Multiply_Float(
     A=&RandomRangeFloat(Minimum=800, Maximum=1200)
     B=&UniformAOrBFloat(A=-1, Evaluation Type=::NewEnumerator0, Recalculate Random Each Loop=true, Override Seed=false)
    ))
   /SetVariables(Particles.Position=&AddVectorToPosition(
     Position=Linked() {linkedVariable {name=System.BladesOffset, type=Vector3f}}
     Vector=&MakeVector(
      X=&Sine(
       Normalized Angle=$Particles.NormalizedAge
       Period=-0.5
       Scale=$Particles.StartRadius
       Bias=&RandomRangeFloat(Minimum=0, Maximum=1)
      )
      Y=0
      Z=&Cosine(
       Normalized Angle=$Particles.NormalizedAge
       Period=-0.5
       Scale=$Particles.StartRadius
       Bias=&RandomRangeFloat(Minimum=0, Maximum=1)
      )
     )
    )) {isSetParameters=true}
  ]
  rendererClasses [&NiagaraSpriteRendererProperties]
  renderers [{rendererClass=&NiagaraSpriteRendererProperties}]
 }
 Windwhirl {
  emitterUpdateScript [
   &EmitterState()
   &SpawnRate(SpawnRate=25, Use Spawn Probability=false)
  ]
  particleSpawnScript [
   &InitializeParticle(
    Color=<rgba 0.75 0.75 1 1>
    Lifetime Max=0.6
    Lifetime Min=0.3
    Lifetime Mode=::NewEnumerator1
    Mesh Scale Mode=::NewEnumerator1
    Mesh Uniform Scale Max=0.17
    Mesh Uniform Scale Min=0.15
    Position Offset=$System.BladesOffset
    Position Offset Coordinate Space=::Local
    UsePositionOffset=true
    Write Material Random=true
   )
   &ShapeLocation(
    Cylinder Height=10
    Cylinder Radius=5
    MaskSpawnGroup=false
    Override Seed=false
    Rotation Angle=90
    Rotation Axis=<0 1 0>
    Rotation Mode=::NewEnumerator1
    Shape Primitive=::NewEnumerator1
    Surface Only=false
    UseOwnerScale=true
   ) {enabled=false}
   &InitialMeshOrientation(
    Mesh Axis To Orient=<0 1 0>
    Mesh Orientation Mode=::NewEnumerator2
    Orientation Vector=<0 0 1>
    Rotation=&RandomRangeVector(Minimum=<0 0 0>, Maximum=<0 0 1>)
    Use Rotation Vector=true
   )
  ]
  particleUpdateScript [
   &ParticleState()
   &ScaleColor(
    Scale Alpha=&FloatFromCurve(
     FloatCurve=DataInterface() {propertyValues {Curve {keys [
         {arriveTangent=9.0102, interpMode=RCIM_Cubic, leaveTangent=9.0102, tangentMode=RCTM_User, time=0, value=0}
         {arriveTangent=2.4457, interpMode=RCIM_Cubic, leaveTangent=2.4457, tangentMode=RCTM_User, time=0.1, value=0.75}
         {arriveTangentWeight=0.1333, interpMode=RCIM_Cubic, leaveTangentWeight=0.1667, tangentMode=RCTM_User, time=0.5, value=1}
         {arriveTangent=-3.2028, interpMode=RCIM_Cubic, leaveTangent=-3.2028, tangentMode=RCTM_User, time=1, value=0}
        ]}}}
     Scale Curve=&FNTimeOfDayScalar(Night Scale=0.1, Morning Scale=0.08, Day Scale=0.08, Evening Scale=0.05)
    )
    Scale RGB=&MakeVectorFromLinearColorRGB(Color=&MakeCustomLinearColorFromBool(
      Bool=$User.bIsFrontEndPreview
      True Color=<rgba 0.8052 0.8383 1 1>
      False Color=&ColorFromCurve(
       ColorCurve=DataInterface() {propertyValues {
         AlphaCurve {keys [<at 0 1 RCIM_Cubic>]}
         BlueCurve {keys [<at 0.07 1 RCIM_Cubic>,<at 0.125 0.6>,<at 0.25 0.7>,<at 0.5 1>,<at 0.625 0.9>,<at 0.75 0.9891>,<at 1 1 RCIM_Cubic>]}
         GreenCurve {keys [<at 0.07 0.4436 RCIM_Cubic>,<at 0.125 0.7169>,<at 0.25 0.8561>,<at 0.5 1>,<at 0.625 0.9522>,<at 0.75 0.8087>,<at 1 0.4436 RCIM_Cubic>]}
         RedCurve {keys [<at 0.07 0.04 RCIM_Cubic>,<at 0.125 1>,<at 0.25 1>,<at 0.5 1>,<at 0.625 1>,<at 0.75 1>,<at 1 0.04 RCIM_Cubic>]}
        }}
       CurveIndex=Linked() {linkedVariable {name=NPC.FortniteNPC.TimeOfDay_0-1, type=NiagaraFloat}}
      )
     ))
    ScaleA=true
    ScaleRGB=true
   )
   &UpdateMeshOrientation(
    Rotation Coordinate Space=::NewEnumerator3
    Rotation Rate=&RandomRangeFloat(Minimum=5, Maximum=10)
    Rotation Vector=&RandomRangeVector(Minimum=<0 0 0>, Maximum=<0 0 1>)
   )
   &ScaleMeshSize(Scale Factor=&Multiply_VectorByFloat(
     Vector=<1 1 1>
     Float=&FloatFromCurve(FloatCurve=DataInterface() {propertyValues {Curve {keys [{arriveTangent=3.0713, interpMode=RCIM_Cubic, leaveTangent=3.0713, tangentMode=RCTM_User, time=0, value=0}, <at 0.6 1 RCIM_Cubic>]}}})
    ))
  ]
  rendererClasses [&NiagaraMeshRendererProperties]
  renderers [{rendererClass=&NiagaraMeshRendererProperties, rendererIndex=0}]
 }
}
summary {
 systemName=NS_Guitar_GoneFishing_GrassExhaust
 emitters [
  {emitterName=GrassMesh, rendererClasses [&NiagaraMeshRendererProperties]}
  {emitterName=GrassGPU, simTarget=GPUComputeSim, rendererClasses [&NiagaraSpriteRendererProperties]}
  {bEnabled=false, emitterName=FN_Debug, rendererClasses [&NiagaraSpriteRendererProperties]}
  {emitterName=Windwhirl, rendererClasses [&NiagaraMeshRendererProperties]}
  {emitterName=GrassSpinning, simTarget=GPUComputeSim, rendererClasses [&NiagaraSpriteRendererProperties]}
  {emitterName=ExhaustSmoke, rendererClasses [&NiagaraSpriteRendererProperties]}
 ]
 userVariables [User.bIsFrontEndPreview {type=NiagaraBool}]
}
