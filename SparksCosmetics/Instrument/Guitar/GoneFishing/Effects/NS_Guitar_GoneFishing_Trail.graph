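# Niagara trail effect for the GoneFishing guitar cosmetic, combining two ribbon trails with flying grass debris.
# Audio_Chroma_Trail (red) and Trail_Instrument_Color (green) use AnimTrail modules, narrowing and fading over short lifetimes with time-of-day tint and alpha.
# GPU grass sprites and CPU grass meshes spawn on User.SkeletalMesh, launch along System.ExhaustFacing, then tumble under gravity, drag and curl noise.
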NiagaraSystem /SparksCosmetics/Instrument/Guitar/GoneFishing/Effects/NS_Guitar_GoneFishing_Trail
dependencies {
 usedDataInterfaces [&NiagaraDataInterfaceCurve, &NiagaraDataInterfaceColorCurve]
 usedDynamicInputs []
 usedModules [&SystemState]
}
emitters {
 Audio_Chroma_Trail {
  emitterUpdateScript [
   &EmitterState()
   &FN_SpawnAnimTrail_v2()
  ]
  particleSpawnScript [
   /Game/Effects/Niagara/Modules/AnimTrail/FN_InitializeAnimTrail/FN_InitializeAnimTrailBoneIndecies(First Bone or Socket Index=1)
   /Game/Effects/Niagara/Modules/AnimTrail/FN_InitializeAnimTrail/FN_InitializeAnimTrail(AnimTrailWidth=1, Color=<rgba 0.5198 0.0035 0 1>, Lifetime=0.25)
   &SystemLocation(MaskSpawnGroup=false)
   &DynamicMaterialParameters(
    Index 0 Param 1=1
    Index 0 Param 4=1
    Module.Index0Param0Write=true
    Module.Index0Param1Write=true
    Module.Index0Param2Write=true
    Module.Index0Param3Write=true
   )
   &AddVelocity(Velocity=<500 0 0>) {enabled=false}
  ]
  particleUpdateScript [
   &ScaleRibbonWidth(Ribbon Width Scale=&FloatFromCurve(FloatCurve=DataInterface() {propertyValues {Curve {keys [
         {arriveTangent=-1.2579, interpMode=RCIM_Cubic, time=0, value=1}
         {arriveTangent=-0.9664, interpMode=RCIM_Cubic, leaveTangent=-0.9664, tangentMode=RCTM_User, time=1, value=0}
        ]}}}))
   &ScaleColor(
    Scale Alpha=&Multiply_Float(
     A=&FloatFromCurve(FloatCurve=DataInterface() {propertyValues {Curve {keys [
          {interpMode=RCIM_Cubic, leaveTangent=-0.143, tangentMode=RCTM_Break, time=0, value=1}
          <at 0.6689 0.0233>
          {arriveTangent=-0.2139, interpMode=RCIM_Cubic, tangentMode=RCTM_Break, time=1, value=0}
         ]}}})
     B=&FNTimeOfDayScalar(Night Scale=1, Morning Scale=0.5, Evening Scale=0.5)
    )
    Scale RGB=<500 500 500>
    ScaleA=true
    ScaleRGB=true
   ) {enabled=false, module=ScaleColor002}
   &FN_UpdateAnimTrail_v2()
   &ScaleColor(
    Scale Alpha=&Multiply_Float(
     A=&FloatFromCurve(FloatCurve=DataInterface() {propertyValues {Curve {keys [
          {interpMode=RCIM_Cubic, leaveTangent=-0.143, tangentMode=RCTM_Break, time=0, value=1}
          <at 0.6689 0.0233>
          {arriveTangent=-0.2139, interpMode=RCIM_Cubic, tangentMode=RCTM_Break, time=1, value=0}
         ]}}})
     B=&FNTimeOfDayScalar(Night Scale=0.65, Morning Scale=0.5, Day Scale=0.7, Evening Scale=0.7)
    )
    Scale RGB=&FN_Select_Vector_from_Time_Of_Day(Morning=<0.9 0.9 0.9>, Day=<1 1 1>, Evening=<1 1 1>, Night=<0.4 0.4 1>)
    ScaleA=true
    ScaleRGB=true
   )
   &SolveForcesAndVelocity(Clamp Velocity=false, Limit Acceleration=false)
  ]
  rendererClasses [&NiagaraRibbonRendererProperties]
  renderers [{rendererClass=&NiagaraRibbonRendererProperties}]
 }
 GrassGPU {
  simTarget=GPUComputeSim
  emitterUpdateScript [
   &EmitterState()
   &SpawnRate(SpawnRate=75, Use Spawn Probability=false)
  ]
  particleSpawnScript [
   /SetVariables(Particles.SubImageIndex=&RandomRangeFloat(Minimum=0, Maximum=6)) {isSetParameters=true}
   &InitializeParticle(
    Color=<rgba 0.6531 0.6531 0.6531 1>
    Lifetime=&UniformRangedFloat(Minimum=0.4, Maximum=0.6)
    Position Offset=$System.ExhaustOffset
    Position Offset Coordinate Space=::Local
    Sprite Rotation Mode=::NewEnumerator2
    Sprite Size Mode=::NewEnumerator1
    Sprite UV Mode=::NewEnumerator3
    Uniform Sprite Size Max=15
    Uniform Sprite Size Min=9
    UsePositionOffset=true
    Write Material Random=true
   )
   &AddVelocity(Velocity=&Multiply_VectorByFloat(Vector=$System.ExhaustFacing, Float=200))
   &InheritVelocity(Module.Limit Speed=true, Velocity Scale=&RandomRangeVector(Minimum=<0.5 0.5 0.5>, Maximum=<1 1 1>))
   &AddVelocityInCone(
    Cone Angle=150
    Cone Axis=$System.ExhaustFacing
    Module.Use Velocity Falloff On Cone Axis=false
    Velocity Strength=&UniformRangedFloat(Minimum=0, Maximum=100)
   )
   &SkeletalMeshLocation(MaskSpawnGroup=false, Mesh Sampling Type=::NewEnumerator5, Skeletal Mesh=$User.SkeletalMesh, Socket Sampling Mode=::NewEnumerator5)
   &ShapeLocation(MaskSpawnGroup=false, Override Seed=false, Sphere Radius=3, UseOwnerScale=true)
   &AddVelocity(Velocity=<500 0 0>) {enabled=false, module=AddVelocity001}
  ]
  particleUpdateScript [
   &ParticleState()
   &Drag(Use Linear Drag=true, Use Rotational Drag=false)
   &GravityForce(Gravity=<0 0 -720>)
   &CurlNoiseForce(Noise Frequency=100, Noise Strength=100, Pan Noise=false)
   &SolveForcesAndVelocity(Clamp Velocity=false, Limit Acceleration=false)
   &ScaleColor(
    Scale Alpha=1.2
    Scale RGB=&MakeVectorFromLinearColorRGB(Color=&ColorFromCurve(ColorCurve=DataInterface() {propertyValues {
        AlphaCurve {keys []}
        BlueCurve {keys [<at 0.0081 0.6244>,<at 0.1423 0.6305>,<at 0.9559 1>]}
        GreenCurve {keys [<at 0.0081 0.8353>,<at 0.1423 0.7629>,<at 0.9559 0.8821>]}
        RedCurve {keys [<at 0.0081 1>,<at 0.1423 1>,<at 0.9559 0.8145>]}
       }}))
    ScaleA=true
    ScaleRGB=true
   )
   &ScaleSpriteSize(Uniform Curve Sprite Scale=DataInterface() {propertyValues {Curve {keys [
        {arriveTangent=8.5765, interpMode=RCIM_Cubic, leaveTangent=8.5765, tangentMode=RCTM_User, time=0.1, value=0.5}
        {arriveTangent=-0.5096, interpMode=RCIM_Cubic, leaveTangent=-0.5096, time=0.2, value=1}
        {arriveTangent=-5.3839, interpMode=RCIM_Cubic, leaveTangent=-5.3839, tangentMode=RCTM_User, time=0.7622, value=0.1625}
        {arriveTangent=0.095, interpMode=RCIM_Cubic, leaveTangent=0.095, tangentMode=RCTM_User, time=0.8, value=-0.05}
       ]}}})
   &SpriteRotationRate(Rotation Rate=&RandomRangeFloat(Minimum=-360, Maximum=360))
  ]
  rendererClasses [&NiagaraSpriteRendererProperties]
  renderers [{rendererClass=&NiagaraSpriteRendererProperties}]
 }
 GrassMesh {
  emitterUpdateScript [
   &EmitterState()
   &SpawnRate(SpawnRate=25, Use Spawn Probability=false)
  ]
  particleSpawnScript [
   /SetVariables(Particles.MeshIndex=&RandomRangeInt(Minimum=0, Maximum=1)) {isSetParameters=true}
   &InitializeParticle(
    Color=&RandomRangeLinearColor(Minimum=<rgba 0.65 1 0.2601 1>, Maximum=<rgba 0.7164 1 0.2601 1>)
    Lifetime=&UniformRangedFloat(Minimum=0.25, Maximum=0.5)
    Mesh Scale Max=<0.2 0.3 0.3>
    Mesh Scale Min=<0.1 0.2 0.2>
    Mesh Scale Mode=::NewEnumerator2
    Position Offset=$System.ExhaustOffset
    Position Offset Coordinate Space=::Local
    UsePositionOffset=true
    Write Material Random=true
   )
   &AddVelocity(Velocity=&Multiply_VectorByFloat(Vector=$System.ExhaustFacing, Float=200))
   &InheritVelocity(
    Module.Limit Speed=true
    Velocity Limit=50
    Velocity Scale=&RandomRangeVector(Minimum=<0.5 0.5 0.5>, Maximum=<1 1 1>)
   )
   &AddVelocityInCone(
    Cone Angle=70
    Cone Axis=$System.ExhaustFacing
    Module.Use Velocity Falloff On Cone Axis=false
    Velocity Strength=&UniformRangedFloat(Minimum=0, Maximum=50)
   )
   &InitialMeshOrientation(Use Rotation Vector=false)
   &SkeletalMeshLocation(MaskSpawnGroup=false, Mesh Sampling Type=::NewEnumerator5, Skeletal Mesh=$User.SkeletalMesh, Socket Sampling Mode=::NewEnumerator5)
   &ShapeLocation(MaskSpawnGroup=false, Override Seed=false, Sphere Radius=3, UseOwnerScale=true)
   &AddVelocity(Velocity=<500 0 0>) {enabled=false, module=AddVelocity001}
  ]
  particleUpdateScript [
   &ParticleState()
   &Drag(Use Linear Drag=true, Use Rotational Drag=false)
   &GravityForce(Gravity=<0 0 -720>)
   &CurlNoiseForce(Noise Frequency=100, Noise Strength=50, Pan Noise=false)
   &SolveForcesAndVelocity(Clamp Velocity=false, Limit Acceleration=false)
   &ScaleColor(
    Scale Alpha=1.2
    Scale RGB=&MakeVectorFromLinearColorRGB(Color=&ColorFromCurve(ColorCurve=DataInterface() {propertyValues {
        AlphaCurve {keys []}
        BlueCurve {keys [<at 0.0081 0.6244>,<at 0.1423 0.6305>,<at 0.9559 1>]}
        GreenCurve {keys [<at 0.0081 0.8353>,<at 0.1423 0.7629>,<at 0.9559 0.8821>]}
        RedCurve {keys [<at 0.0081 1>,<at 0.1423 1>,<at 0.9559 0.8145>]}
       }}))
    ScaleA=true
    ScaleRGB=true
   )
   &UpdateMeshOrientation(Rotation Rate=&RandomRangeFloat(Minimum=-0.4, Maximum=0.4), Rotation Vector=<1 1 1>)
   &ScaleMeshSize(Scale Factor=&VectorFromFloat(Value=&FloatFromCurve(FloatCurve=DataInterface() {propertyValues {Curve {keys [
          {arriveTangent=8.5765, interpMode=RCIM_Cubic, leaveTangent=8.5765, tangentMode=RCTM_User, time=0.1, value=0.7}
          {arriveTangent=0.65, interpMode=RCIM_Cubic, leaveTangent=0.65, time=0.2, value=1}
          {arriveTangent=-1.25, interpMode=RCIM_Cubic, leaveTangent=-1.25, time=0.6, value=0.625}
          {arriveTangent=0.095, interpMode=RCIM_Cubic, leaveTangent=0.095, tangentMode=RCTM_User, time=1, value=0}
         ]}}})))
   &ScaleSpriteSizeBySpeed(
    Max Scale Factor=&Vector2DFromFloat(Value=1)
    Min Scale Factor=&Vector2DFromFloat(Value=0.25)
    Velocity Threshold=200
   )
  ]
  rendererClasses [&NiagaraMeshRendererProperties]
  renderers [{rendererClass=&NiagaraMeshRendererProperties, rendererIndex=0}]
 }
 Trail_Instrument_Color {
  emitterUpdateScript [
   &EmitterState()
   &FN_SpawnAnimTrail_v2()
  ]
  particleSpawnScript [
   &FN_InitializeAnimTrailBoneIndecies(First Bone or Socket Index=1)
   &FN_InitializeAnimTrail(AnimTrailWidth=1, Color=<rgba 0.0616 0.3406 0.0132 0.4>, Lifetime=0.2)
   &SystemLocation(MaskSpawnGroup=false)
   &AddVelocity(Velocity=<500 0 0>) {enabled=false}
  ]
  particleUpdateScript [
   &ScaleRibbonWidth(Ribbon Width Scale=&FloatFromCurve(FloatCurve=DataInterface() {propertyValues {Curve {keys [
         {arriveTangent=-1.2579, interpMode=RCIM_Cubic, time=0, value=1}
         {arriveTangent=-0.9664, interpMode=RCIM_Cubic, leaveTangent=-0.9664, tangentMode=RCTM_User, time=1, value=0}
        ]}}}))
   &ScaleColor(
    Scale Alpha=&Multiply_Float(
     A=&FloatFromCurve(FloatCurve=DataInterface() {propertyValues {Curve {keys [
          {interpMode=RCIM_Cubic, leaveTangent=-2.3646, tangentMode=RCTM_Break, time=0, value=1}
          <at 0.6689 0.0233>
          {arriveTangent=-0.2139, interpMode=RCIM_Cubic, tangentMode=RCTM_Break, time=1, value=0}
         ]}}})
     B=&FNTimeOfDayScalar(Night Scale=0.5, Morning Scale=0.65, Day Scale=0.7, Evening Scale=0.7)
    )
    Scale RGB=&Multiply_VectorByFloat(
     Vector=&FN_Select_Vector_from_Time_Of_Day(Morning=<0.9 0.9 0.9>, Day=<1 1 1>, Evening=<1 1 1>, Night=<0.5 0.5 1>)
     Float=&FloatFromCurve(FloatCurve=DataInterface() {propertyValues {Curve {keys [
          {interpMode=RCIM_Cubic, leaveTangent=-11.0297, tangentMode=RCTM_Break, time=0, value=1}
          {arriveTangent=-0.2139, interpMode=RCIM_Cubic, tangentMode=RCTM_Break, time=0.1, value=0}
         ]}}})
    )
    ScaleA=true
    ScaleRGB=true
   ) {module=ScaleColor002}
   &FN_UpdateAnimTrail_v2()
   &SolveForcesAndVelocity(Clamp Velocity=false, Limit Acceleration=false)
  ]
  rendererClasses [&NiagaraRibbonRendererProperties]
  renderers [{rendererClass=&NiagaraRibbonRendererProperties}]
 }
}
summary {
 systemName=NS_Guitar_GoneFishing_Trail
 emitters [
  {emitterName=Trail_Instrument_Color, rendererClasses [&NiagaraRibbonRendererProperties]}
  {emitterName=Audio_Chroma_Trail, rendererClasses [&NiagaraRibbonRendererProperties]}
  {emitterName=GrassMesh, rendererClasses [&NiagaraMeshRendererProperties]}
  {emitterName=GrassGPU, simTarget=GPUComputeSim, rendererClasses [&NiagaraSpriteRendererProperties]}
 ]
 userVariables [
  User.RibbonIDOffset {type=NiagaraInt32}
  User.SkeletalMesh {
   type=NiagaraDataInterfaceSkeletalMesh
   defaultValue {
    struct=NiagaraExt_VariableValue_DataInterface
    value {dataInterface=&self:NiagaraDataInterfaceSkeletalMesh_0}
   }
  }
 ]
}
