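# Physics asset for the GoneFishing guitar cosmetic in Sparks, giving its moving accessory parts capsule-based physics bodies.
# The slurp can and both wheel bars joint to the handle body, with the can's twist locked and the wheels left one swing axis to rotate on.
# A pull cord chains from the handle through two rope segments to the pull handle, with swing and twist limit angles set at each joint.
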PhysicsAsset /SparksCosmetics/Instrument/Guitar/GoneFishing/Meshes/GoneFishing_Guitar_Physics
bodies {
 handle [
  {shapeType=Capsule, length=6.2205, radius=3.2007, shapeName=handle_capsule, center=<-1.851 4.154 -3.3147>}
  {shapeType=Capsule, length=4.8885, radius=1.7124, shapeName=handle_capsule, center=<-1.6257 2.3349 4.2998>, rotation=<rot 0.0002 0 -90.4566>}
 ]
 pull_handle [
  {shapeType=Capsule, length=2, radius=0.5, shapeName=pull_handle_capsule, center=<0.9468 0 0>, rotation=<rot 90 0 0>}
  {shapeType=Capsule, length=4.0929, radius=1.0428, shapeName=pull_handle_capsule, center=<2.7318 0.0843 -0.1408>, rotation=<rot 0 0 -12.5833>}
 ]
 rope_01 [{shapeType=Capsule, length=0.25, radius=0.5, shapeName=rope_01_capsule, rotation=<rot 90 0 0>}]
 rope_02 [{shapeType=Capsule, length=0.25, radius=0.5, shapeName=rope_02_capsule, rotation=<rot 90 0 0>}]
 slurp_can [{shapeType=Capsule, length=2.116, radius=3.0069, shapeName=slurp_can_capsule, rotation=<rot 8 0 -12>}]
 wheels_bottom [{shapeType=Capsule, length=15, radius=1, shapeName=wheels_bottom_capsule, rotation=<rot 0 0 90>}]
 wheels_top [{shapeType=Capsule, length=15, radius=1, shapeName=wheels_top_capsule, rotation=<rot 0 0 90>}]
}
constraints [
 {bone1Name=slurp_can, bone2Name=handle, swing1LimitDegrees=3, swing2LimitDegrees=3, twistLimitDegrees=45, twistMotion=Locked}
 {
  bone1Name=wheels_bottom
  bone2Name=handle
  swing1LimitDegrees=45
  swing1Motion=Locked
  swing2LimitDegrees=180
  twistLimitDegrees=0
  twistMotion=Locked
 }
 {
  bone1Name=wheels_top
  bone2Name=handle
  swing1LimitDegrees=45
  swing1Motion=Locked
  swing2LimitDegrees=180
  twistLimitDegrees=0
  twistMotion=Locked
 }
 {bone1Name=rope_01, bone2Name=handle, swing1LimitDegrees=15, swing2LimitDegrees=40, twistLimitDegrees=4}
 {bone1Name=rope_02, bone2Name=rope_01, swing1LimitDegrees=20, swing2LimitDegrees=35, twistLimitDegrees=4}
 {bone1Name=pull_handle, bone2Name=rope_02, swing1LimitDegrees=20, swing2LimitDegrees=45, twistLimitDegrees=10}
]
