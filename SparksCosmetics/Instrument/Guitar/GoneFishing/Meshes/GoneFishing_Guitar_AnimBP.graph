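# Animation Blueprint for the GoneFishing guitar cosmetic, animating a vibrating motor and spinning saw blade.
# AnimGraph runs handle-relative RigidBody physics, then additive bone edits for motor parts and blade.
# Each update builds sine-wave jitter from game time and accumulates blade angle from RotationSpeed.
# Speed is high in frontend or while a delayed flag, driven by the harvesting montage or pickaxe special action, is set; otherwise it settles lower.
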AnimBlueprint /SparksCosmetics/Instrument/Guitar/GoneFishing/Meshes/GoneFishing_Guitar_AnimBP
target_skeleton=&/GoneFishing_Guitar_Skeleton
graphs [
 AnimGraph {nodes [
   n1 = RigidBody(ComponentPose=Local To Component(LocalPose=Input Pose())) {
    bClampLinearTranslationLimitToRefPose=true
    bOverrideWorldGravity=true
    overridePhysicsAsset=&/GoneFishing_Guitar_Physics
    simulationSpace=BaseBoneSpace
    alphaBoolBlend {blendOption=Linear}
    baseBoneRef {boneName=handle}
    componentLinearAccScale=<0.4 0.3 0.3>
    componentLinearVelScale=<0.4 0.3 0.3>
    overrideWorldGravity {z=-980}
    simSpaceSettings {
     dampingAlpha=0
     maxAngularAcceleration=80
     maxAngularVelocity=6
     maxLinearAcceleration=800
     maxLinearVelocity=500
     worldAlpha=1
     externalLinearDragV=<0.1 0.1 0.1>
    }
   }
   n2 = "Transform (Modify) Bone - Bone: motor"(
    ComponentPose=@n1
    Alpha=1
    Translation=Get MotorTranslate() -> (MotorTranslate:Vector)
    Rotation=Get MotorRotate() -> (MotorRotate:Rotator)
    Scale="1,1,1"
   ) {
    rotationMode=BMM_Additive
    rotationSpace=BCS_BoneSpace
    translationMode=BMM_Additive
    translationSpace=BCS_BoneSpace
    alphaBoolBlend {blendOption=Linear}
    boneToModify {boneName=motor}
   }
   n3 = Get Low Motor Translate() -> (Low Motor Translate:Vector)
   n4 = "Transform (Modify) Bone - Bone: slurp_can"(ComponentPose=@n2, Alpha=1, Translation=@n3, Rotation=Get MotorRotate(), Scale="1,1,1") {
    rotationMode=BMM_Additive
    rotationSpace=BCS_BoneSpace
    translationMode=BMM_Additive
    translationSpace=BCS_BoneSpace
    alphaBoolBlend {blendOption=Linear}
    boneToModify {boneName=slurp_can}
   }
   n5 = Get Low Motor Rotate() -> (Low Motor Rotate:Rotator)
   n6 = "Transform (Modify) Bone - Bone: wheels_bottom"(ComponentPose=@n4, Alpha=1, Translation=@n3, Rotation=@n5, Scale="1,1,1") {
    rotationMode=BMM_Additive
    rotationSpace=BCS_BoneSpace
    translationMode=BMM_Additive
    translationSpace=BCS_BoneSpace
    alphaBoolBlend {blendOption=Linear}
    boneToModify {boneName=wheels_bottom}
   }
   n7 = "Transform (Modify) Bone - Bone: wheels_top"(ComponentPose=@n6, Alpha=1, Translation=@n3, Rotation=@n5, Scale="1,1,1") {
    rotationMode=BMM_Additive
    rotationSpace=BCS_BoneSpace
    translationMode=BMM_Additive
    translationSpace=BCS_BoneSpace
    alphaBoolBlend {blendOption=Linear}
    boneToModify {boneName=wheels_top}
   }
   n8 = "Transform (Modify) Bone - Bone: rope_base"(ComponentPose=@n7, Alpha=1, Translation=Get MotorTranslate(), Rotation=Get MotorRotate(), Scale="1,1,1") {
    rotationMode=BMM_Additive
    rotationSpace=BCS_BoneSpace
    translationMode=BMM_Additive
    translationSpace=BCS_BoneSpace
    alphaBoolBlend {blendOption=Linear}
    boneToModify {boneName=rope_base}
   }
   n9 = "Transform (Modify) Bone - Bone: blade"!(
    ComponentPose=@n8
    Alpha=1
    Translation="0,0,0"
    Rotation_Roll:float=0
    Rotation_Pitch:float=Get!() -> (rotation of saw blade_Roll:float, rotation of saw blade_Pitch:float, rotation of saw blade_Yaw:float).rotation of saw blade_Pitch
    Rotation_Yaw:float=0
    Scale="1,1,1"
   ) -> (Pose) {
    rotationMode=BMM_Additive
    rotationSpace=BCS_BoneSpace
    translationSpace=BCS_BoneSpace
    alphaBoolBlend {blendOption=Linear}
    boneToModify {boneName=blade}
   }
   Output Pose(Result=Component To Local(ComponentPose=@n9))
   Reroute Node(InputPin:Wildcard) -> (OutputPin:Wildcard)
  ]}
 EventGraph {nodes [
   Event BlueprintUpdateAnimation()
   > n1 = Sequence!(execute) -> (then_0, then_1, then_2:Exec, then_3:Exec, then_4:Exec, then_5:Exec)
   n2 = Branch(execute=@n1.then_1, Condition=Get animation plus delay() -> (animation plus delay:bool))
   Set RotationSpeed(execute=[@n3.then, @n2.then], RotationSpeed:double=4500) -> (Output_Get:double)
   n3 = Branch(execute=@n2.else, Condition=Get bIsInFrontEnd())
   n4 = GetWorldDeltaSeconds!() -> (ReturnValue:double)
   n5 = Get RotationSpeed() -> (RotationSpeed:double)
   Set RotationSpeed(execute=@n3.else, RotationSpeed="Clamp (Float)"(Value=FInterpTo(Current=@n5, Target=1000, DeltaTime=@n4, InterpSpeed=1), Min=1000, Max=2000))
   n6 = GetAnimInstance(self=GetOwnerSkeletalMeshComponent())
   Set is Animation player melee combo(
    execute=@n1.then_3
    is Animation player melee combo:bool="Equal (Object)"(
     A=GetCurrentActiveMontage!(self:Anim Instance Object Reference=@n6) -> (ReturnValue:Anim Montage Object Reference)
     B=/Game/Animation/Game/MainPlayer/Combat/Melee/TwoHanded/NarrowGrip/UpperBody/BaseBall_Bat/BaseballBat_Harvesting_Combo_M
    )
   ) -> (Output_Get:bool)
   n7 = GetGameTimeInSeconds()
   n8 = "Sin (Radians)"(A=float * float(A=float * float(A=1, B=60), B=@n7))
   Set MotorTranslate(execute=@n1.then_4, MotorTranslate:Vector=MakeVector(Y=0, Z=float * float(A=@n8, B=0.08))) -> (Output_Get:Vector)
   n9 = "Sin (Radians)"(A=float * float(A=float * float(A=1, B=40), B=@n7))
   n10 = "Sin (Radians)"(A=float * float(A=float * float(A=1, B=40), B=@n7))
   > Set MotorRotate(MotorRotate:Rotator=MakeRotator(Pitch=float * float(A=@n10, B=0.5), Yaw=float * float(A=@n9, B=0.6))) -> (Output_Get:Rotator)
   n11 = "Sin (Radians)"(A=float * float(A=float * float(A=1, B=60), B=@n7))
   > Set Low Motor Translate(Low Motor Translate:Vector=MakeVector(Y=0, Z=float * float(A=@n11, B=0.02))) -> (Output_Get:Vector)
   n12 = "Sin (Radians)"(A=float * float(A=float * float(A=1, B=18), B=@n7))
   n13 = "Sin (Radians)"(A=float * float(A=float * float(A=1, B=20), B=@n7))
   > Set Low Motor Rotate(Low Motor Rotate:Rotator=MakeRotator(Pitch=float * float(A=@n12, B=0.2), Yaw=float * float(A=@n13, B=0.1))) -> (Output_Get:Rotator)
   n14 = GetWorldDeltaSeconds()
   n15 = Get RotationSpeed()
   n16 = Set CurrentAngle(
    execute=@n1.then_2
    CurrentAngle:double=float + float(A=Get CurrentAngle() -> (CurrentAngle:double), B=float * float(A=@n14, B=@n15) {is=title K2Node_PromotableOperator})
   ) -> (Output_Get:double)
   > n17 = Set CurrentAngle(CurrentAngle="% (Float)"!(A:double=@n16, B:double=360) -> (ReturnValue:double))
   > Set rotation of saw blade(rotation of saw blade:Rotator=MakeRotator(Pitch=@n17)) -> (Output_Get:Rotator)
   n18 = Branch(execute=@n1.then_0, Condition=Get bIsInFrontEnd())
   n19 = Branch(execute=@n18.else, Condition=Get is Animation player melee combo() -> (is Animation player melee combo:bool))
   > Delay(Duration=1)
   > Set animation plus delay(animation plus delay:bool=true) -> (Output_Get:bool)
   Delay(execute=@n19.else, Duration=1)
   > Set animation plus delay(animation plus delay=false)
   Event OnPickaxeSpecialAction!() -> (OutputDelegate, then, ActionName:Name=None) {is=title K2Node_Event}
   > Delay(Duration=0.85)
   > Set animation plus delay(animation plus delay=true)
   > Delay(Duration=3)
   > Set animation plus delay(animation plus delay=false)
   TryGetPawnOwner()
  ]}
]
variables [
 animation plus delay {type=boolean}
 CurrentAngle {type=number}
 is Animation player melee combo {type=boolean}
 Low Motor Rotate {type=Rotator}
 Low Motor Translate {type=Vector}
 MotorRotate {type=Rotator}
 MotorTranslate {type=Vector}
 Rotation {type=Rotator}
 rotation of saw blade {type=Rotator}
 RotationSpeed {type=number}
]
